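# Niagara system for the idle foot effect of the YukonGladTube skin of a cosmetic shoe.
# Only FN_AnimTrail is enabled: it spawns two short-lived ribbon trails from skeletal mesh bone or socket pairs, scaling width per foot from the left and right shoe-scale parameters.
# Its ribbon material gets a near-black time-of-day tint, panning opacity and erosion over lifetime.
# Black ribbon and grey splash sprite emitters, modulated by a system timeline, remain in the system but are disabled.
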NiagaraSystem /CosmeticShoes/Assets/FORT_Shoes/Shoe_YukonGladRusset/Skins/YukonGladTube/FX/NS_Shoe_YukonGladTube_Idle
dependencies {
 usedDataInterfaces [
  &NiagaraDataInterfaceCurve
  &NiagaraDataInterfaceColorCurve
  &NiagaraDataInterfaceSpriteRendererInfo
 ]
 usedDynamicInputs [
  /Game/Effects/Niagara/DynamicInputs/Logic/FN_IsRunningCheck
  &Boolean_AndOperation
  &Boolean_NotOperation
 ]
 usedModules [
  &self:SystemSpawnScript.SystemScriptSource.SystemScriptGraph.NiagaraNodeAssignment_0.SetVariables_#1
  &self:SystemSpawnScript.SystemScriptSource.SystemScriptGraph.NiagaraNodeAssignment_1.SetVariables_#2
  &SystemState
  &self:SystemSpawnScript.SystemScriptSource.SystemScriptGraph.NiagaraNodeAssignment_2.SetVariables_#3
  &self:SystemSpawnScript.SystemScriptSource.SystemScriptGraph.NiagaraNodeAssignment_4.SetVariables_#4
  &Timeline
  &self:SystemSpawnScript.SystemScriptSource.SystemScriptGraph.NiagaraNodeAssignment_3.SetVariables_#5
 ]
}
emitters {
 BlackFillTrail {
  bEnabled=false
  emitterUpdateScript [
   &EmitterState()
   &FN_SpawnRibbon_v2(Number of Ribbons=2, Spawn Mode=::NewEnumerator0, SpawnRate=64)
   &FN_Integrator() {enabled=false}
  ]
  particleSpawnScript [
   /Game/Effects/Niagara/Modules/Ribbons/FN_InitializeRibbon_v4(
    Color:LinearColor=<rgba 0.0052 0.0049 0.0045 2>
    Enable Skeletal Position=true
    Lifetime=0.2
    Ribbon Width=30
    SkeletalMesh:NiagaraDataInterfaceSkeletalMesh=$System.Skel2
   )
   &FN_FlipRibbonV()
   &AddVelocity(Velocity=<0 -130 0>)
  ]
  particleUpdateScript [
   &SolveForcesAndVelocity(Clamp Velocity=false, Limit Acceleration=false)
   &InheritSourceMovement(
    Applied Position Scale=&VectorFromFloat(Value=&FloatFromCurve(FloatCurve=DataInterface() {propertyValues {Curve {keys [
          {interpMode=RCIM_Cubic, tangentMode=RCTM_User, time=0, value=1}
          {arriveTangent=-0.0657, interpMode=RCIM_Cubic, leaveTangent=-0.0657, tangentMode=RCTM_User, time=1, value=0.14}
         ]}}}))
    ApplyForce=false
    ApplyPosition=true
    ApplyPositionFalloff=false
   ) {enabled=false}
   /Game/Effects/Niagara/Modules/Ribbons/FN_UpdateRibbon_v4(Enable Skeletal Position=true, SkeletalMesh:NiagaraDataInterfaceSkeletalMesh=$System.Skel2)
   &FN_RibbonMaterial() {enabled=false}
   &ScaleColor(
    Scale Alpha=&FloatFromCurve(
     FloatCurve=DataInterface() {propertyValues {Curve {keys [{interpMode=RCIM_Cubic, leaveTangent=-3.1111, tangentMode=RCTM_Break, time=0, value=1}, <at 1 0>]}}}
     Scale Curve=$System.Timeline.Time
    )
    Scale RGB=&MakeVectorFromLinearColorRGB(Color=&ColorFromCurve(ColorCurve=DataInterface() {propertyValues {
        AlphaCurve {keys []}
        BlueCurve {keys [<at 0.0628 1 RCIM_Cubic>,<at 0.8496 0.1453 RCIM_Cubic>]}
        GreenCurve {keys [<at 0.0628 1 RCIM_Cubic>,<at 0.8496 0.1534 RCIM_Cubic>]}
        RedCurve {keys [<at 0.0628 1 RCIM_Cubic>,<at 0.8496 0.1684 RCIM_Cubic>]}
       }}))
    ScaleA=true
    ScaleRGB=true
   )
   &ScaleRibbonWidth(Ribbon Width Scale=&FloatFromCurve(FloatCurve=DataInterface() {propertyValues {Curve {keys [
         {interpMode=RCIM_Cubic, leaveTangent=0.6658, tangentMode=RCTM_Break, time=0, value=1}
         {arriveTangent=1.4789, tangentMode=RCTM_User, time=1, value=2}
        ]}}}))
   &DynamicMaterialParameters(
    Index 0 Param 1=&FloatFromCurve(FloatCurve=DataInterface() {propertyValues {Curve {keys [<at 0 1>,<at 0.5 0.3>,<at 1 0.05>]}}})
    Index 0 Param 2=0.2
    Index 0 Param 3=-2
    Index 0 Param 4=0.7
    Module.Index0Float0Write=true
    Module.Index0Float1Write=true
    Module.Index0Float2Write=true
    Module.Index0Float3Write=true
   )
   &Color(Color=&MultiplyLinearColor(Linear Color=$Particles.Color, Scale Factor=&MakeVector4_FromLinearColor(LinearColor=$System.TODColor)))
  ]
  rendererClasses [&NiagaraRibbonRendererProperties]
  renderers [{rendererClass=&NiagaraRibbonRendererProperties}]
 }
 BlackFillTrail001 {
  bEnabled=false
  emitterUpdateScript [
   &EmitterState()
   &FN_SpawnRibbon_v2(Number of Ribbons=2, Spawn Mode=::NewEnumerator0, SpawnRate=64)
   &FN_Integrator() {enabled=false}
  ]
  particleSpawnScript [
   &FN_InitializeRibbon_v4(
    Color:LinearColor=<rgba 0.0052 0.0049 0.0045 2>
    Enable Skeletal Position=true
    Lifetime=0.2
    Ribbon Width=15
    SkeletalMesh:NiagaraDataInterfaceSkeletalMesh=$System.Skel2
   )
   &FN_FlipRibbonV()
   &AddVelocity(Velocity=<0 -130 0>)
  ]
  particleUpdateScript [
   &SolveForcesAndVelocity(Clamp Velocity=false, Limit Acceleration=false)
   &InheritSourceMovement(
    Applied Position Scale=&VectorFromFloat(Value=&FloatFromCurve(FloatCurve=DataInterface() {propertyValues {Curve {keys [
          {interpMode=RCIM_Cubic, tangentMode=RCTM_User, time=0, value=1}
          {arriveTangent=-0.0657, interpMode=RCIM_Cubic, leaveTangent=-0.0657, tangentMode=RCTM_User, time=1, value=0.14}
         ]}}}))
    ApplyForce=false
    ApplyPosition=true
    ApplyPositionFalloff=false
   ) {enabled=false}
   &FN_UpdateRibbon_v4(Enable Skeletal Position=true, SkeletalMesh:NiagaraDataInterfaceSkeletalMesh=$System.Skel2)
   &FN_RibbonMaterial() {enabled=false}
   &ScaleColor(
    Scale Alpha=&FloatFromCurve(
     FloatCurve=DataInterface() {propertyValues {Curve {keys [{interpMode=RCIM_Cubic, leaveTangent=-3.1111, tangentMode=RCTM_Break, time=0, value=1}, <at 1 0>]}}}
     Scale Curve=$System.Timeline.Time
    )
    Scale RGB=&MakeVectorFromLinearColorRGB(Color=&ColorFromCurve(ColorCurve=DataInterface() {propertyValues {
        AlphaCurve {keys []}
        BlueCurve {keys [<at 0.0628 1 RCIM_Cubic>,<at 0.8496 0.1453 RCIM_Cubic>]}
        GreenCurve {keys [<at 0.0628 1 RCIM_Cubic>,<at 0.8496 0.1534 RCIM_Cubic>]}
        RedCurve {keys [<at 0.0628 1 RCIM_Cubic>,<at 0.8496 0.1684 RCIM_Cubic>]}
       }}))
    ScaleA=true
    ScaleRGB=true
   )
   &ScaleRibbonWidth(Ribbon Width Scale=&FloatFromCurve(FloatCurve=DataInterface() {propertyValues {Curve {keys [
         {interpMode=RCIM_Cubic, leaveTangent=0.6658, tangentMode=RCTM_Break, time=0, value=1}
         {arriveTangent=1.4789, tangentMode=RCTM_User, time=1, value=2}
        ]}}}))
   &DynamicMaterialParameters(
    Index 0 Param 1=&FloatFromCurve(FloatCurve=DataInterface() {propertyValues {Curve {keys [<at 0 1>,<at 0.5 0.3>,<at 1 0.05>]}}})
    Index 0 Param 2=0.2
    Index 0 Param 3=-2
    Index 0 Param 4=0.7
    Module.Index0Float0Write=true
    Module.Index0Float1Write=true
    Module.Index0Float2Write=true
    Module.Index0Float3Write=true
   )
   &Color(Color=&MultiplyLinearColor(Linear Color=$Particles.Color, Scale Factor=&MakeVector4_FromLinearColor(LinearColor=$System.TODColor)))
  ]
  rendererClasses [&NiagaraRibbonRendererProperties]
  renderers [
   {rendererClass=&NiagaraRibbonRendererProperties}
   {rendererClass=&NiagaraRibbonRendererProperties, rendererIndex=1}
  ]
 }
 FN_AnimTrail {
  emitterSpawnScript [&FN_RibbonIDOffsetUtility()]
  emitterUpdateScript [
   &EmitterState()
   &FN_SpawnAnimTrail(Number of Anim Trails=2, Switch on ENiagara_AnimTrail_Source:ENiagara_AnimTrail_Source=::NewEnumerator1)
  ]
  particleSpawnScript [
   &FN_InitializeAnimTrailBoneIndecies()
   &FN_InitializeAnimTrailBoneIndecies(AnimTrail Selection=1, First Bone or Socket Index=2, Second Bone or Socket Index=3) {module=FN_InitializeAnimTrailBoneIndecies001}
   &FN_InitializeAnimTrail(AnimTrailWidth=2.5, Lifetime=0.16, SkeletalMesh=$System.Skeletal Mesh)
   &AddVelocity(Velocity=<0 400 0>) {enabled=false}
  ]
  particleUpdateScript [
   &SolveForcesAndVelocity(Clamp Velocity=false, Limit Acceleration=false)
   &FN_UpdateAnimTrail(SkeletalMesh=$System.Skeletal Mesh) {module=FN_UpdateAnimTrail001}
   &ScaleColor(Scale Mode=::NewEnumerator2) {enabled=false}
   &DynamicMaterialParameters(
    Index 0 Param 1=0.8
    Index 0 Param 2=1
    Index 0 Param 3=-0.2
    Index 0 Param 4=1
    Module.Index0Float0Write=true
    Module.Index0Float1Write=true
    Module.Index0Float2Write=true
    Module.Index0Float3Write=true
   )
   &FN_RibbonMaterial(
    Color=&FNTimeOfDayColor(
     Night Color=<rgba 0.0087 0.0112 0.0191 1>
     Morning Color=<rgba 0.0191 0.0133 0.0138 1>
     Day Color=<rgba 0.03 0.03 0.03 1>
     Evening Color=<rgba 0.03 0.03 0.03 1>
    )
    Opacity Erosion:NiagaraFloat=&FloatFromCurve(FloatCurve=DataInterface() {propertyValues {Curve {
        defaultValue=0
        keys [{arriveTangent=1.1228, interpMode=RCIM_Cubic, leaveTangent=1.1228, tangentMode=RCTM_User, time=0, value=0}, <at 1 1 RCIM_Cubic>]
       }}})
    Opacity Pan U Speed=5
    Opacity Power=0.5
    Scale Alpha=1.5
   )
   &ScaleColor(Scale Alpha=$System.Timeline.Time, ScaleA=true, ScaleRGB=true) {enabled=false, module=ScaleColor001}
   &ScaleRibbonWidth(Ribbon Width Scale=&MaskFloatBySpawnGroup(
     Masked Float=&MakeFloatFromVector(Vector=$User.ShoeScale_L, Channel=::NewEnumerator0)
     Passthrough Float=&MakeFloatFromVector(Vector=$User.ShoeScale_R, Channel=::NewEnumerator0)
    ))
  ]
  rendererClasses [&NiagaraRibbonRendererProperties]
  renderers [
   {rendererClass=&NiagaraRibbonRendererProperties}
   {rendererClass=&NiagaraRibbonRendererProperties, rendererIndex=1}
  ]
 }
 Splashes {
  bEnabled=false
  emitterUpdateScript [
   &EmitterState()
   &SpawnRate(
    SpawnRate=&FloatFromCurve(
     FloatCurve=DataInterface() {propertyValues {Curve {keys [<at 0 0>,<at 0.4 0>,<at 0.9 1>]}}}
     CurveIndex=$System.Timeline.Time
     Scale Curve=55
    )
    Use Spawn Probability=false
   ) {module=SpawnRate001}
   &SpawnRate(SpawnRate=50, Use Spawn Probability=false) {enabled=false}
   &SpawnPerUnit(
    Module.Use Max Movement Threshold=true
    Module.Use Movement Tolerance=false
    Spawn Spacing=25
    Use Spawn Probability=false
    Velocity Vector:Vector3f=&Multiply_Vector(A=$Engine.Owner.Velocity, B=<1 1 0>)
   ) {enabled=false}
  ]
  particleSpawnScript [
   &PartitionParticles()
   &InitializeParticle(
    Color=<rgba 0.4983 0.4983 0.4983 1>
    Lifetime Max=0.4
    Lifetime Min=0.2
    Lifetime Mode=::NewEnumerator1
    Sprite Rotation Mode=::NewEnumerator2
    Sprite Size Mode=::NewEnumerator1
    Uniform Sprite Size Max=45
    Uniform Sprite Size Min=25
    UsePositionOffset=false
    Write Material Random=true
   )
   &SkeletalMeshLocation(
    Bone / Socket Index=$Particles.PartitionParticles.Partition
    Bone Sampling Mode=::NewEnumerator4
    MaskSpawnGroup=false
    Sampled Position Offset=<0 0 -5>
    Skeletal Mesh=$System.Skel2
   )
   &ShapeLocation(MaskSpawnGroup=false, Override Seed=false, Sphere Radius=5, Sphere Surface Distribution=0.3452, UseOwnerScale=true)
   &AddVelocity(Velocity=&RandomRangeVector(Minimum=<0 0 40>, Maximum=<0 0 120>)) {enabled=false}
   &InheritVelocity(Inherited Velocity Amount Scale=<-0.4 -0.4 0>, Speed Limit=true, UseSpeedThreshold=false)
   &InheritVelocity(Inherited Velocity Amount Scale=<1 1 0>, Speed Limit=true, UseSpeedThreshold=false) {enabled=false, module=InheritVelocity001}
   &Color(Color=&MultiplyLinearColor(Linear Color=$Particles.Color, Scale Factor=&MakeVector4_FromLinearColor(LinearColor=$System.TODColor)))
   &SubUVAnimation(
    "Sprite Renderer "=DataInterface() {propertyValues {SpriteRenderer=&self:Splashes_0.NiagaraSpriteRendererProperties_0}}
    SubUV Animation Mode:ENiagara_SubUVLookupModeV2=::NewEnumerator2
    UseEndFrame=false
    UseStartFrame=false
   )
  ]
  particleUpdateScript [
   &ParticleState()
   &ScaleSpriteSize(Uniform Curve Sprite Scale=DataInterface() {propertyValues {Curve {keys [
        {interpMode=RCIM_Cubic, leaveTangent=4.0079, tangentMode=RCTM_Break, time=0, value=0}
        {interpMode=RCIM_Cubic, tangentMode=RCTM_Break, time=0.3, value=0.85}
        {arriveTangent=0.452, tangentMode=RCTM_User, time=1, value=1.2}
       ]}}})
   &Drag(Drag=4, Use Linear Drag=true, Use Rotational Drag=true)
   &GravityForce(Gravity=&RandomRangeVector(Minimum=<0 0 -450>, Maximum=<0 0 -700>)) {enabled=false}
   &SolveForcesAndVelocity(Clamp Velocity=false, Limit Acceleration=false)
   &ScaleSpriteSizeBySpeed(Max Scale Factor=<1.2 5>, Min Scale Factor=<0.7 2>)
   &DynamicMaterialParameters(
    Index 0 Param 1=&RandomRangeFloat(Minimum=0, Maximum=1)
    Index 0 Param 2=&FloatFromCurve(FloatCurve=DataInterface() {propertyValues {Curve {
        defaultValue=1
        keys [
         {interpMode=RCIM_Cubic, leaveTangent=1.6626, tangentMode=RCTM_Break, time=0, value=0}
         {arriveTangent=-0.2778, tangentMode=RCTM_User, time=1, value=1}
        ]
       }}})
    Index 0 Param 3=1
    Index 0 Param 4=&RandomRangeFloat(Minimum=1, Maximum=2)
    Module.Index0Float0Write=true
    Module.Index0Float1Write=true
    Module.Index0Float2Write=true
    Module.Index0Float3Write=true
   )
   &ScaleColor(
    Linear Color Curve=DataInterface() {propertyValues {
      AlphaCurve {keys [<at 0.2568 1>,<at 0.9652 0>]}
      BlueCurve {keys [<at 0 0.6328>,<at 0.0106 0.8131>,<at 0.8297 0.283>]}
      GreenCurve {keys [<at 0 0.7071>,<at 0.0106 0.8131>,<at 0.8297 0.283>]}
      RedCurve {keys [<at 0 0.8142>,<at 0.0106 0.8131>,<at 0.8297 0.283>]}
     }}
    Scale Mode=::NewEnumerator2
   )
  ]
  rendererClasses [&NiagaraSpriteRendererProperties]
  renderers [{rendererClass=&NiagaraSpriteRendererProperties}]
 }
 Splashes002 {
  bEnabled=false
  emitterUpdateScript [
   &EmitterState()
   &SpawnRate(
    SpawnRate=&FloatFromCurve(
     FloatCurve=DataInterface() {propertyValues {Curve {keys [<at 0 0>,<at 0.4 0>,<at 0.9 1>]}}}
     CurveIndex=$System.Timeline.Time
     Scale Curve=55
    )
    Use Spawn Probability=false
   ) {module=SpawnRate001}
   &SpawnRate(SpawnRate=50, Use Spawn Probability=false) {enabled=false}
   &SpawnPerUnit(
    Module.Use Max Movement Threshold=true
    Module.Use Movement Tolerance=false
    Spawn Spacing=25
    Use Spawn Probability=false
    Velocity Vector:Vector3f=&Multiply_Vector(A=$Engine.Owner.Velocity, B=<1 1 0>)
   ) {enabled=false}
  ]
  particleSpawnScript [
   &PartitionParticles()
   &InitializeParticle(
    Color=<rgba 0.4983 0.4983 0.4983 1>
    Lifetime Max=0.4
    Lifetime Min=0.2
    Lifetime Mode=::NewEnumerator1
    Sprite Rotation Mode=::NewEnumerator2
    Sprite Size Mode=::NewEnumerator1
    Uniform Sprite Size Max=8
    Uniform Sprite Size Min=4
    UsePositionOffset=false
    Write Material Random=true
   )
   &SkeletalMeshLocation(
    Bone / Socket Index=$Particles.PartitionParticles.Partition
    Bone Sampling Mode=::NewEnumerator4
    MaskSpawnGroup=false
    Sampled Position Offset=<0 0 -5>
    Skeletal Mesh=$System.Skel2
   )
   &ShapeLocation(MaskSpawnGroup=false, Override Seed=false, Sphere Radius=5, Sphere Surface Distribution=0.3452, UseOwnerScale=true)
   &AddVelocity(Velocity=&RandomRangeVector(Minimum=<0 0 40>, Maximum=<0 0 120>)) {enabled=false}
   &InheritVelocity(Inherited Velocity Amount Scale=<-0.4 -0.4 0>, Speed Limit=true, UseSpeedThreshold=false)
   &InheritVelocity(Inherited Velocity Amount Scale=<1 1 0>, Speed Limit=true, UseSpeedThreshold=false) {enabled=false, module=InheritVelocity001}
   &Color(Color=&MultiplyLinearColor(Linear Color=$Particles.Color, Scale Factor=&MakeVector4_FromLinearColor(LinearColor=$System.TODColor)))
   &SubUVAnimation(
    "Sprite Renderer "=DataInterface() {propertyValues {SpriteRenderer=&self:Splashes002_0.NiagaraSpriteRendererProperties_0}}
    SubUV Animation Mode:ENiagara_SubUVLookupModeV2=::NewEnumerator2
    UseEndFrame=false
    UseStartFrame=false
   )
  ]
  particleUpdateScript [
   &ParticleState()
   &ScaleSpriteSize(Uniform Curve Sprite Scale=DataInterface() {propertyValues {Curve {keys [
        {interpMode=RCIM_Cubic, leaveTangent=4.0079, tangentMode=RCTM_Break, time=0, value=0}
        {interpMode=RCIM_Cubic, tangentMode=RCTM_Break, time=0.3, value=1}
        {arriveTangent=-0.5586, interpMode=RCIM_Cubic, leaveTangent=-5.1823, tangentMode=RCTM_Break, time=0.5, value=0.85}
        {arriveTangent=0.452, tangentMode=RCTM_User, time=1, value=0}
       ]}}})
   &Drag(Drag=4, Use Linear Drag=true, Use Rotational Drag=true)
   &GravityForce(Gravity=&RandomRangeVector(Minimum=<0 0 -450>, Maximum=<0 0 -700>)) {enabled=false}
   &SolveForcesAndVelocity(Clamp Velocity=false, Limit Acceleration=false)
   &ScaleSpriteSizeBySpeed(Max Scale Factor=<1.2 5>, Min Scale Factor=<0.7 2>) {enabled=false}
   &DynamicMaterialParameters(
    Index 0 Param 1=&RandomRangeFloat(Minimum=0, Maximum=1)
    Index 0 Param 2=&FloatFromCurve(FloatCurve=DataInterface() {propertyValues {Curve {
        defaultValue=1
        keys [
         {interpMode=RCIM_Cubic, leaveTangent=0.8069, tangentMode=RCTM_Break, time=0, value=0}
         {arriveTangent=-0.2778, tangentMode=RCTM_User, time=1, value=1}
        ]
       }}})
    Index 0 Param 3=1
    Module.Index0Float0Write=true
    Module.Index0Float1Write=true
    Module.Index0Float2Write=true
    Module.Index0Float3Write=true
   )
   &ScaleColor(
    Linear Color Curve=DataInterface() {propertyValues {
      AlphaCurve {keys [<at 0.2568 1>,<at 0.9652 0>]}
      BlueCurve {keys [<at 0 0.6328>,<at 0.0106 0.8131>,<at 0.8297 0.283>]}
      GreenCurve {keys [<at 0 0.7071>,<at 0.0106 0.8131>,<at 0.8297 0.283>]}
      RedCurve {keys [<at 0 0.8142>,<at 0.0106 0.8131>,<at 0.8297 0.283>]}
     }}
    Scale Mode=::NewEnumerator2
   )
  ]
  rendererClasses [&NiagaraSpriteRendererProperties]
  renderers [{rendererClass=&NiagaraSpriteRendererProperties}]
 }
}
summary {
 systemName=NS_Shoe_YukonGladTube_Idle
 emitters [
  {bEnabled=false, emitterName=BlackFillTrail, rendererClasses [&NiagaraRibbonRendererProperties]}
  {bEnabled=false, emitterName=Splashes, rendererClasses [&NiagaraSpriteRendererProperties]}
  {bEnabled=false, emitterName=BlackFillTrail001, rendererClasses [&NiagaraRibbonRendererProperties]}
  {bEnabled=false, emitterName=Splashes002, rendererClasses [&NiagaraSpriteRendererProperties]}
  {emitterName=FN_AnimTrail, rendererClasses [&NiagaraRibbonRendererProperties]}
 ]
 userVariables [
  User.HideFootFX {type=NiagaraBool}
  User.IsSkydiving {type=NiagaraBool}
  User.RibbonIDOffset {type=NiagaraInt32}
  User.ShoeScale_L {type=Vector3f, description="", defaultValue {struct=Vector3f, value=<1 1 1>}}
  User.ShoeScale_R {type=Vector3f, description="", defaultValue {struct=Vector3f, value=<1 1 1>}}
  User.SkeletalMesh {
   type=NiagaraDataInterfaceSkeletalMesh
   defaultValue {
    struct=NiagaraExt_VariableValue_DataInterface
    value {dataInterface=&self:NiagaraDataInterfaceSkeletalMesh_0}
   }
  }
 ]
}
